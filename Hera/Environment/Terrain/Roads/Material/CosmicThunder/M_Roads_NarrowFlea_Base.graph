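# Base material for NarrowFlea roads that renders striped grass blended into the landscape.
# MF_Hera_Grass takes biome-derived terrain masks, then stripe masks and MF_NarrowFlea_Colors tint its base color and subsurface.
# Edge falloff masks, optional fade-out and height lerps build road opacity for runtime virtual texture output, with switchable RVT readback.
# Mobile uses masked opacity and a camera-directed world position offset, replaced by a downward offset for shadows.
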Material /Hera/Environment/Terrain/Roads/Material/CosmicThunder/M_Roads_NarrowFlea_Base
graph {
 expressions {
  MaterialExpressionAbs_7 {class=MaterialExpressionAbs, inputs [None {from=MaterialExpressionConstantBiasScale_0, output=""}], properties {}}
  MaterialExpressionAbs_8 {class=MaterialExpressionAbs, inputs [None {from=MaterialExpressionConstantBiasScale_1, output=""}], properties {}}
  MaterialExpressionAbs_9 {class=MaterialExpressionAbs, inputs [None {from=MaterialExpressionScalarParameter_12, output=""}], properties {}}
  MaterialExpressionAdd_0 {
   class=MaterialExpressionAdd
   inputs [
    A {from=MaterialExpressionMultiply_47}
    B {from=MaterialExpressionVectorParameter_0, output=Add}
   ]
  }
  MaterialExpressionAdd_1 {
   class=MaterialExpressionAdd
   inputs [
    A {from=MaterialExpressionNamedRerouteUsage_8}
    B {from=""}
   ]
  }
  MaterialExpressionAdd_5 {
   class=MaterialExpressionAdd
   inputs [
    A {from=MaterialExpressionMultiply_17}
    B {from=MaterialExpressionVectorParameter_4, output=Add}
   ]
  }
  MaterialExpressionAdd_6 {
   class=MaterialExpressionAdd
   inputs [
    A {from=""}
    B {from=MaterialExpressionMax_1}
   ]
   properties {constA=0.5}
  }
  MaterialExpressionAppendVector_0 {
   class=MaterialExpressionAppendVector
   inputs [
    A {from=MaterialExpressionConvert_2, output=""}
    B {from=MaterialExpressionMax_2, output=""}
   ]
  }
  MaterialExpressionAppendVector_1 {
   class=MaterialExpressionAppendVector
   inputs [
    A {from=MaterialExpressionLinearInterpolate_0, output=""}
    B {from=MaterialExpressionConvert_5, output=""}
   ]
  }
  MaterialExpressionCameraPositionWS_0 {class=MaterialExpressionCameraPositionWS, inputs [], properties {}}
  MaterialExpressionComponentMask_3 {class=MaterialExpressionComponentMask, inputs [None {from=MaterialExpressionTextureCoordinate_1}], properties {g=true}}
  MaterialExpressionConstant3Vector_3 {class=MaterialExpressionConstant3Vector, properties {constant=<rgba 0 0 -60 0>}}
  MaterialExpressionConstantBiasScale_0 {class=MaterialExpressionConstantBiasScale, inputs [None {from=MaterialExpressionConvert_3}], properties {bias=-0.5, scale=2}}
  MaterialExpressionConstantBiasScale_1 {class=MaterialExpressionConstantBiasScale, inputs [None {from=MaterialExpressionFrac_0}], properties {bias=-0.5, scale=2}}
  MaterialExpressionConstant_1 {class=MaterialExpressionConstant, properties {r=1}}
  MaterialExpressionConstant_24 {class=MaterialExpressionConstant, properties {r=1}}
  MaterialExpressionConstant_41 {class=MaterialExpressionConstant, properties {r=1}}
  MaterialExpressionConstant_8 {class=MaterialExpressionConstant, properties {r=1}}
  MaterialExpressionConvert_0 {
   class=MaterialExpressionConvert
   inputs [None {from=MaterialExpressionTextureCoordinate_0}]
   properties {
    nodeName=Break Float2
    convertInputs [{
      type=Vector2
      expressionInput {
       expression=&self:MaterialExpressionTextureCoordinate_0
       inputName=None
       mask=0
       maskA=0
       maskB=0
       maskG=0
       maskR=0
       outputIndex=0
      }
     }]
    convertOutputs [{type=Vector2}]
   }
  }
  MaterialExpressionConvert_1 {
   class=MaterialExpressionConvert
   inputs [
    None {from=""}
    None {from=""}
    None {from=MaterialExpressionAdd_1}
    None {from=""}
   ]
   properties {
    nodeName=Make Vector4
    convertInputs [
     {type=Scalar, expressionInput {expression=null, inputName=None, mask=0, maskA=0, maskB=0, maskG=0, maskR=0, outputIndex=0}}
     {type=Scalar, expressionInput {expression=null, inputName=None, mask=0, maskA=0, maskB=0, maskG=0, maskR=0, outputIndex=0}}
     {
      type=Scalar
      expressionInput {
       expression=&self:MaterialExpressionAdd_1
       inputName=None
       mask=0
       maskA=0
       maskB=0
       maskG=0
       maskR=0
       outputIndex=0
      }
     }
     {type=Scalar, expressionInput {expression=null, inputName=None, mask=0, maskA=0, maskB=0, maskG=0, maskR=0, outputIndex=0}}
    ]
    convertOutputs [{type=Vector4, defaultValue=<rgba 0 0 0 1>}]
   }
  }
  MaterialExpressionConvert_2 {
   class=MaterialExpressionConvert
   inputs [None {from=MaterialExpressionMaterialFunctionCall_25, output=QuadC_RGBA}]
   properties {
    nodeName=Break Float4
    convertInputs [{
      type=Vector4
      defaultValue=<rgba 0 0 0 1>
      expressionInput {
       expression=&self:MaterialExpressionMaterialFunctionCall_25
       inputName=None
       mask=0
       maskA=0
       maskB=0
       maskG=0
       maskR=0
       outputIndex=3
      }
     }]
    convertOutputs [{type=Vector3}, {type=Scalar}]
   }
  }
  MaterialExpressionConvert_24 {
   class=MaterialExpressionConvert
   inputs [
    None {from=""}
    None {from=""}
    None {from=""}
    None {from=""}
   ]
   properties {
    nodeName=Make Vector4
    convertInputs [
     {type=Scalar, expressionInput {expression=null, inputName=None, mask=0, maskA=0, maskB=0, maskG=0, maskR=0, outputIndex=0}}
     {type=Scalar, expressionInput {expression=null, inputName=None, mask=0, maskA=0, maskB=0, maskG=0, maskR=0, outputIndex=0}}
     {type=Scalar, expressionInput {expression=null, inputName=None, mask=0, maskA=0, maskB=0, maskG=0, maskR=0, outputIndex=0}}
     {type=Scalar, expressionInput {expression=null, inputName=None, mask=0, maskA=0, maskB=0, maskG=0, maskR=0, outputIndex=0}}
    ]
    convertOutputs [{type=Vector4, defaultValue=<rgba 0 0 0 1>}]
   }
  }
  MaterialExpressionConvert_25 {
   class=MaterialExpressionConvert
   inputs [
    None {from=""}
    None {from=""}
    None {from=""}
    None {from=""}
   ]
   properties {
    nodeName=Make Vector4
    convertInputs [
     {type=Scalar, expressionInput {expression=null, inputName=None, mask=0, maskA=0, maskB=0, maskG=0, maskR=0, outputIndex=0}}
     {type=Scalar, expressionInput {expression=null, inputName=None, mask=0, maskA=0, maskB=0, maskG=0, maskR=0, outputIndex=0}}
     {
      type=Scalar
      defaultValue=<rgba 1 0 0 1>
      expressionInput {expression=null, inputName=None, mask=0, maskA=0, maskB=0, maskG=0, maskR=0, outputIndex=0}
     }
     {type=Scalar, expressionInput {expression=null, inputName=None, mask=0, maskA=0, maskB=0, maskG=0, maskR=0, outputIndex=0}}
    ]
    convertOutputs [{type=Vector4, defaultValue=<rgba 0 0 0 1>}]
   }
  }
  MaterialExpressionConvert_26 {
   class=MaterialExpressionConvert
   inputs [
    None {from=""}
    None {from=""}
    None {from=""}
    None {from=MaterialExpressionMax_2}
   ]
   properties {
    nodeName=Make Vector4
    convertInputs [
     {type=Scalar, expressionInput {expression=null, inputName=None, mask=0, maskA=0, maskB=0, maskG=0, maskR=0, outputIndex=0}}
     {type=Scalar, expressionInput {expression=null, inputName=None, mask=0, maskA=0, maskB=0, maskG=0, maskR=0, outputIndex=0}}
     {type=Scalar, expressionInput {expression=null, inputName=None, mask=0, maskA=0, maskB=0, maskG=0, maskR=0, outputIndex=0}}
     {
      type=Scalar
      expressionInput {
       expression=&self:MaterialExpressionMax_2
       inputName=None
       mask=0
       maskA=0
       maskB=0
       maskG=0
       maskR=0
       outputIndex=0
      }
     }
    ]
    convertOutputs [{type=Vector4, defaultValue=<rgba 0 0 0 1>}]
   }
  }
  MaterialExpressionConvert_27 {
   class=MaterialExpressionConvert
   inputs [
    None {from=MaterialExpressionLinearInterpolate_0}
    None {from=""}
    None {from=""}
    None {from=""}
   ]
   properties {
    nodeName=Make Vector4
    convertInputs [
     {
      type=Scalar
      expressionInput {
       expression=&self:MaterialExpressionLinearInterpolate_0
       inputName=None
       mask=0
       maskA=0
       maskB=0
       maskG=0
       maskR=0
       outputIndex=0
      }
     }
     {type=Scalar, expressionInput {expression=null, inputName=None, mask=0, maskA=0, maskB=0, maskG=0, maskR=0, outputIndex=0}}
     {type=Scalar, expressionInput {expression=null, inputName=None, mask=0, maskA=0, maskB=0, maskG=0, maskR=0, outputIndex=0}}
     {type=Scalar, expressionInput {expression=null, inputName=None, mask=0, maskA=0, maskB=0, maskG=0, maskR=0, outputIndex=0}}
    ]
    convertOutputs [{type=Vector4, defaultValue=<rgba 0 0 0 1>}]
   }
  }
  MaterialExpressionConvert_3 {
   class=MaterialExpressionConvert
   inputs [None {from=MaterialExpressionTextureCoordinate_4}]
   properties {
    nodeName=Break Float2
    convertInputs [{
      type=Vector2
      expressionInput {
       expression=&self:MaterialExpressionTextureCoordinate_4
       inputName=None
       mask=0
       maskA=0
       maskB=0
       maskG=0
       maskR=0
       outputIndex=0
      }
     }]
    convertOutputs [{type=Scalar}, {type=Scalar}]
   }
  }
  MaterialExpressionConvert_5 {
   class=MaterialExpressionConvert
   inputs [None {from=MaterialExpressionMaterialFunctionCall_25, output=QuadD_RGBA}]
   properties {
    nodeName=Break Float4
    convertInputs [{
      type=Vector4
      defaultValue=<rgba 0 0 0 1>
      expressionInput {
       expression=&self:MaterialExpressionMaterialFunctionCall_25
       inputName=None
       mask=0
       maskA=0
       maskB=0
       maskG=0
       maskR=0
       outputIndex=4
      }
     }]
    convertOutputs [{type=Scalar}, {type=Vector3}]
   }
  }
  MaterialExpressionDistance_0 {
   class=MaterialExpressionDistance
   inputs [
    A {from=MaterialExpressionWorldPosition_1, output=XYZ}
    B {from=MaterialExpressionCameraPositionWS_0, output=""}
   ]
   properties {}
  }
  MaterialExpressionDivide_1 {
   class=MaterialExpressionDivide
   inputs [
    A {from=MaterialExpressionDistance_0}
    B {from=""}
   ]
   properties {constB=700}
  }
  MaterialExpressionDotProduct_0 {
   class=MaterialExpressionDotProduct
   inputs [
    A {from=MaterialExpressionMaterialFunctionCall_12, output=Result}
    B {from=MaterialExpressionMaterialFunctionCall_27, output=Result}
   ]
   properties {}
  }
  MaterialExpressionDotProduct_1 {
   class=MaterialExpressionDotProduct
   inputs [
    A {from=MaterialExpressionTextureSample_1, output=RGBA}
    B {from=MaterialExpressionMaterialFunctionCall_12, output=Result}
   ]
   properties {}
  }
  MaterialExpressionFrac_0 {class=MaterialExpressionFrac, inputs [None {from=MaterialExpressionMultiply_19, output=""}], properties {}}
  MaterialExpressionGetMaterialAttributes_3 {
   class=MaterialExpressionGetMaterialAttributes
   inputs [None {from=MaterialExpressionMaterialFunctionCall_3, output=StaticMeshMaterial}]
   properties {attributeGetTypes [#1, #2]}
  }
  MaterialExpressionGetMaterialAttributes_8 {
   class=MaterialExpressionGetMaterialAttributes
   inputs [None {from=MaterialExpressionNamedRerouteUsage_7, output=""}]
   properties {attributeGetTypes [
     #1
     #3
     #4
     #5
    ]}
  }
  MaterialExpressionIf_0 {
   class=MaterialExpressionIf
   inputs [
    A {from=MaterialExpressionNamedRerouteUsage_123, output=""}
    B {from="", output=""}
    "A > B" {from=MaterialExpressionNamedRerouteUsage_122, output=""}
    "A == B" {from="", output=""}
    "A < B" {from=MaterialExpressionConstant_41, output=""}
   ]
   properties {constB=0.5}
  }
  MaterialExpressionIf_1 {
   class=MaterialExpressionIf
   inputs [
    A {from=MaterialExpressionAbs_9, output=""}
    B {from="", output=""}
    "A > B" {from=MaterialExpressionAbs_9, output=""}
    "A == B" {from="", output=""}
    "A < B" {from=MaterialExpressionConstant_8, output=""}
   ]
   properties {constB=0.0001}
  }
  MaterialExpressionIf_2 {
   class=MaterialExpressionIf
   inputs [
    A {from=MaterialExpressionScalarParameter_3, output=""}
    B {from="", output=""}
    "A > B" {from=MaterialExpressionConvert_0, output=""}
    "A == B" {from="", output=""}
    "A < B" {from=MaterialExpressionTextureCoordinate_4, output=""}
   ]
   properties {constB=0.5}
  }
  MaterialExpressionLinearInterpolate_0 {
   class=MaterialExpressionLinearInterpolate
   inputs [
    A {from=MaterialExpressionConvert_5}
    B {from=""}
    Alpha {from=MaterialExpressionMultiply_5}
   ]
   properties {constB=0}
  }
  MaterialExpressionLinearInterpolate_2 {
   class=MaterialExpressionLinearInterpolate
   inputs [
    A {from=MaterialExpressionComponentMask_3}
    B {from=MaterialExpressionOneMinus_1}
    Alpha {from=MaterialExpressionScalarParameter_0}
   ]
  }
  MaterialExpressionLinearInterpolate_4 {
   class=MaterialExpressionLinearInterpolate
   inputs [
    A {from=MaterialExpressionMaterialFunctionCall_1, output=GrassStripeA}
    B {from=MaterialExpressionMaterialFunctionCall_1, output=GrassStripeA}
    Alpha {from=MaterialExpressionDotProduct_1}
   ]
   properties {constA=1}
  }
  MaterialExpressionLinearInterpolate_5 {
   class=MaterialExpressionLinearInterpolate
   inputs [
    A {from=""}
    B {from=MaterialExpressionLinearInterpolate_4}
    Alpha {from=MaterialExpressionNamedRerouteDeclaration_1}
   ]
   properties {constA=1}
  }
  MaterialExpressionMaterialFunctionCall_0 {
   class=MaterialExpressionMaterialFunctionCall
   inputs [
    NoRVT {from=MaterialExpressionReroute_9}
    RVT {from=MaterialExpressionMaterialFunctionCall_5, output=Result}
   ]
   properties {materialFunction=&FN_RVT_Switch_Material}
  }
  MaterialExpressionMaterialFunctionCall_1 {
   class=MaterialExpressionMaterialFunctionCall
   inputs []
   properties {materialFunction=/Hera/Terrain/Cliffs/Materials/CosmicThunder/MaterialFunctions/MF_NarrowFlea_Colors}
  }
  MaterialExpressionMaterialFunctionCall_12 {
   class=MaterialExpressionMaterialFunctionCall
   inputs [
    X {from=MaterialExpressionVectorParameter_8, output=R}
    Y {from=MaterialExpressionVectorParameter_8, output=R}
    Z {from=MaterialExpressionVectorParameter_8, output=R}
    A {from=MaterialExpressionVectorParameter_8, output=R}
   ]
   properties {materialFunction=&MakeFloat4}
  }
  MaterialExpressionMaterialFunctionCall_2 {
   class=MaterialExpressionMaterialFunctionCall
   inputs []
   properties {materialFunction=&MF_NarrowFlea_Colors}
  }
  MaterialExpressionMaterialFunctionCall_20 {
   class=MaterialExpressionMaterialFunctionCall
   inputs [
    Diffuse 1 {from=""}
    Diffuse 2 {from=""}
    Transistion Phase {from=MaterialExpressionOneMinus_18}
    Height Texture 1 {from=MaterialExpressionNamedRerouteUsage_26}
    Height Texture 2 {from=MaterialExpressionNamedRerouteUsage_40}
    Contrast {from=MaterialExpressionScalarParameter_1}
   ]
   properties {materialFunction=/Engine/Functions/Engine_MaterialFunctions02/Texturing/HeightLerpWithTwoHeightMaps}
  }
  MaterialExpressionMaterialFunctionCall_22 {
   class=MaterialExpressionMaterialFunctionCall
   inputs [
    A {from=""}
    B {from=""}
    Transition Phase {from=MaterialExpressionReroute_16}
    Height Texture {from=MaterialExpressionNamedRerouteUsage_121}
    Contrast {from=MaterialExpressionScalarParameter_1}
   ]
   properties {materialFunction=&HeightLerp}
  }
  MaterialExpressionMaterialFunctionCall_25 {
   class=MaterialExpressionMaterialFunctionCall
   inputs [
    UseLandscapeCoords {from=""}
    "UseProjection(translucent)" {from=""}
   ]
   properties {materialFunction=/BRCoreMaterials/Hera/Terrain/BiomeMask/MF_Hera_BiomeMasks}
  }
  MaterialExpressionMaterialFunctionCall_27 {
   class=MaterialExpressionMaterialFunctionCall
   inputs [
    X {from=MaterialExpressionTextureSampleParameter2D_7, output=R}
    Y {from=MaterialExpressionTextureSampleParameter2D_7, output=R}
    Z {from=MaterialExpressionTextureSampleParameter2D_7, output=R}
    A {from=MaterialExpressionMultiply_18}
   ]
   properties {materialFunction=&MakeFloat4}
  }
  MaterialExpressionMaterialFunctionCall_3 {
   class=MaterialExpressionMaterialFunctionCall
   inputs [
    Material {from=""}
    MobileMaterial {from=""}
    IsForStaticMesh {from=""}
    UseTangentNormals {from=MaterialExpressionStaticBool_0}
    UseCustomSlopeMask {from=""}
    CustomSlopeMask {from=""}
    Grass_EdgeOnly {from=""}
    LayerWeight {from=MaterialExpressionConstant_1}
    TerrainMasks_QuadA {from=MaterialExpressionNamedRerouteUsage_6}
    TerrainMasks_QuadB {from=MaterialExpressionNamedRerouteUsage_9}
    TerrainMasks_QuadC {from=MaterialExpressionNamedRerouteUsage_10}
    TerrainMasks_QuadD {from=MaterialExpressionNamedRerouteUsage_11}
    VariationMasks Mobile {from=MaterialExpressionMaterialFunctionCall_6, output=VariationMasks Mobile}
    DesertGrass_SimpleNoise {from=""}
   ]
   properties {materialFunction=/Hera/Environment/Terrain/Landscape/Materials/Layers/MF_Hera_Grass}
  }
  MaterialExpressionMaterialFunctionCall_5 {
   class=MaterialExpressionMaterialFunctionCall
   inputs [
    RVT {from=MaterialExpressionStaticSwitchParameter_1}
    ByPass RVT {from=MaterialExpressionReroute_9}
   ]
   properties {materialFunction=/Game/Athena/Apollo/Maps/Landscape/VT/MaterialFunctions/EnableRVT}
  }
  MaterialExpressionMaterialFunctionCall_6 {
   class=MaterialExpressionMaterialFunctionCall
   inputs []
   properties {materialFunction=/BRCoreMaterials/Hermes/Terrain/Landscape/Materials/MaterialFunctions/MF_Hermes_VariationMasks_Mobile}
  }
  MaterialExpressionMaterialFunctionCall_7 {
   class=MaterialExpressionMaterialFunctionCall
   inputs []
   properties {materialFunction=&MF_NarrowFlea_Colors}
  }
  MaterialExpressionMaterialProxyReplace_0 {
   class=MaterialExpressionMaterialProxyReplace
   inputs [
    Realtime {from=MaterialExpressionShadingPathSwitch_3, output=""}
    MaterialProxy {from=MaterialExpressionReroute_6, output=""}
   ]
   properties {}
  }
  MaterialExpressionMax_1 {
   class=MaterialExpressionMax
   inputs [
    A {from=MaterialExpressionMultiply_20}
    B {from=""}
   ]
   properties {constB=0.0001}
  }
  MaterialExpressionMax_2 {
   class=MaterialExpressionMax
   inputs [
    A {from=MaterialExpressionConvert_2}
    B {from=MaterialExpressionNamedRerouteUsage_12}
   ]
  }
  MaterialExpressionMultiply_0 {
   class=MaterialExpressionMultiply
   inputs [
    A {from=MaterialExpressionLinearInterpolate_2}
    B {from=""}
   ]
   properties {constB=2}
  }
  MaterialExpressionMultiply_1 {
   class=MaterialExpressionMultiply
   inputs [
    A {from=MaterialExpressionNamedRerouteDeclaration_44}
    B {from=MaterialExpressionNamedRerouteUsage_0}
   ]
  }
  MaterialExpressionMultiply_17 {
   class=MaterialExpressionMultiply
   inputs [
    A {from=MaterialExpressionSmoothStep_3}
    B {from=MaterialExpressionVectorParameter_4, output=Multiply}
   ]
  }
  MaterialExpressionMultiply_18 {
   class=MaterialExpressionMultiply
   inputs [
    A {from=MaterialExpressionTextureSampleParameter2D_7, output=A}
    B {from=MaterialExpressionTextureSampleParameter2D_7, output=A}
   ]
   properties {desc=Should this just be done in the texture itself?}
  }
  MaterialExpressionMultiply_19 {
   class=MaterialExpressionMultiply
   inputs [
    A {from=MaterialExpressionConvert_3}
    B {from=MaterialExpressionScalarParameter_12}
   ]
  }
  MaterialExpressionMultiply_2 {
   class=MaterialExpressionMultiply
   inputs [
    A {from=MaterialExpressionMultiply_1}
    B {from=MaterialExpressionScalarParameter_11}
   ]
  }
  MaterialExpressionMultiply_20 {
   class=MaterialExpressionMultiply
   inputs [
    A {from=MaterialExpressionOneMinus_4}
    B {from=""}
   ]
   properties {constB=0.5}
  }
  MaterialExpressionMultiply_22 {
   class=MaterialExpressionMultiply
   inputs [
    A {from=MaterialExpressionLinearInterpolate_5}
    B {from=MaterialExpressionGetMaterialAttributes_3, output=Base Color}
   ]
  }
  MaterialExpressionMultiply_23 {
   class=MaterialExpressionMultiply
   inputs [
    A {from=MaterialExpressionDivide_1}
    B {from=MaterialExpressionNormalize_0}
   ]
  }
  MaterialExpressionMultiply_3 {
   class=MaterialExpressionMultiply
   inputs [
    A {from=MaterialExpressionNamedRerouteUsage_5}
    B {from=MaterialExpressionNamedRerouteUsage_3}
   ]
  }
  MaterialExpressionMultiply_4 {
   class=MaterialExpressionMultiply
   inputs [
    A {from=MaterialExpressionMultiply_3}
    B {from=MaterialExpressionMaterialFunctionCall_2, output=GrassStripeSSS}
   ]
  }
  MaterialExpressionMultiply_47 {
   class=MaterialExpressionMultiply
   inputs [
    A {from=MaterialExpressionSmoothStep_2}
    B {from=MaterialExpressionVectorParameter_0, output=Multiply}
   ]
  }
  MaterialExpressionMultiply_49 {
   class=MaterialExpressionMultiply
   inputs [
    A {from=MaterialExpressionSaturate_16}
    B {from=MaterialExpressionSaturate_13}
   ]
  }
  MaterialExpressionMultiply_5 {
   class=MaterialExpressionMultiply
   inputs [
    A {from=MaterialExpressionNamedRerouteUsage_13}
    B {from=MaterialExpressionMaterialFunctionCall_7, output=GrassStripeBiomeVariationInfluence}
   ]
  }
  MaterialExpressionMultiply_56 {
   class=MaterialExpressionMultiply
   inputs [
    A {from=MaterialExpressionIf_0}
    B {from=MaterialExpressionNamedRerouteUsage_2}
   ]
  }
  MaterialExpressionNamedRerouteDeclaration_0 {
   class=MaterialExpressionNamedRerouteDeclaration
   inputs [None {from=MaterialExpressionSetMaterialAttributes_4}]
   properties {name=FinalMaterialWithNoRVT, nodeColor=<rgba 0.5 0.5 0 1>}
  }
  MaterialExpressionNamedRerouteDeclaration_1 {
   class=MaterialExpressionNamedRerouteDeclaration
   inputs [None {from=MaterialExpressionSmoothStep_0}]
   properties {name=StripedMask, nodeColor=<rgba 0.6406 0 1 1>}
  }
  MaterialExpressionNamedRerouteDeclaration_101 {
   class=MaterialExpressionNamedRerouteDeclaration
   inputs [None {from=MaterialExpressionConvert_25}]
   properties {name=TerrainMasks - QuadB, nodeColor=<rgba 0.5 0 1 1>}
  }
  MaterialExpressionNamedRerouteDeclaration_102 {
   class=MaterialExpressionNamedRerouteDeclaration
   inputs [None {from=MaterialExpressionConvert_24}]
   properties {name=TerrainMasks - QuadA, nodeColor=<rgba 0.5 0 1 1>}
  }
  MaterialExpressionNamedRerouteDeclaration_103 {
   class=MaterialExpressionNamedRerouteDeclaration
   inputs [None {from=MaterialExpressionConvert_26}]
   properties {name=TerrainMasks - QuadC, nodeColor=<rgba 0.5 0 1 1>}
  }
  MaterialExpressionNamedRerouteDeclaration_104 {
   class=MaterialExpressionNamedRerouteDeclaration
   inputs [None {from=MaterialExpressionConvert_27}]
   properties {name=TerrainMasks - QuadD, nodeColor=<rgba 0.5 0 1 1>}
  }
  MaterialExpressionNamedRerouteDeclaration_16 {
   class=MaterialExpressionNamedRerouteDeclaration
   inputs [None {from=MaterialExpressionSaturate_29}]
   properties {name=FadeOutMask, nodeColor=<rgba 0.5469 0 1 1>}
  }
  MaterialExpressionNamedRerouteDeclaration_21 {
   class=MaterialExpressionNamedRerouteDeclaration
   inputs [None {from=MaterialExpressionMultiply_49}]
   properties {name=RoadType_EdgeGradient, nodeColor=<rgba 0.5469 0 1 1>}
  }
  MaterialExpressionNamedRerouteDeclaration_23 {
   class=MaterialExpressionNamedRerouteDeclaration
   inputs [None {from=MaterialExpressionSetMaterialAttributes_15}]
   properties {name=RVT_READ_Material, nodeColor=<rgba 0.5 0.5 0 1>}
  }
  MaterialExpressionNamedRerouteDeclaration_33 {
   class=MaterialExpressionNamedRerouteDeclaration
   inputs [None {from=MaterialExpressionSmoothStep_4}]
   properties {name=StripedSSSMask, nodeColor=<rgba 0.6406 0 1 1>}
  }
  MaterialExpressionNamedRerouteDeclaration_39 {
   class=MaterialExpressionNamedRerouteDeclaration
   inputs [None {from=MaterialExpressionNamedRerouteDeclaration_42}]
   properties {name=Final_RoadOpacityMask, nodeColor=<rgba 0.5 0.5 0 1>}
  }
  MaterialExpressionNamedRerouteDeclaration_42 {
   class=MaterialExpressionNamedRerouteDeclaration
   inputs [None {from=MaterialExpressionShadingPathSwitch_0}]
   properties {name=Custom01_OpacityMask, nodeColor=<rgba 0 0.4674 1 1>}
  }
  MaterialExpressionNamedRerouteDeclaration_44 {
   class=MaterialExpressionNamedRerouteDeclaration
   inputs [None {from=MaterialExpressionGetMaterialAttributes_3, output=Opacity Mask}]
   properties {name=Custom01_Height, nodeColor=<rgba 0.5469 0 1 1>}
  }
  MaterialExpressionNamedRerouteDeclaration_46 {
   class=MaterialExpressionNamedRerouteDeclaration
   inputs [None {from=MaterialExpressionMultiply_56}]
   properties {name=EdgeGradientOrFadeOut, nodeColor=<rgba 0.5469 0 1 1>}
  }
  MaterialExpressionNamedRerouteDeclaration_52 {
   class=MaterialExpressionNamedRerouteDeclaration
   inputs [None {from=MaterialExpressionStaticBoolParameter_3}]
   properties {name=Use_FalloffToLandscapeLayerMatch, nodeColor=<rgba 0.5469 0 1 1>}
  }
  MaterialExpressionNamedRerouteDeclaration_53 {
   class=MaterialExpressionNamedRerouteDeclaration
   inputs [None {from=MaterialExpressionSaturate_17}]
   properties {name=RoadTypeEdgeFalloffMask_Channel, nodeColor=<rgba 0.5469 0 1 1>}
  }
  MaterialExpressionNamedRerouteDeclaration_54 {
   class=MaterialExpressionNamedRerouteDeclaration
   inputs [None {from=MaterialExpressionScalarParameter_2}]
   properties {name=Use_FadeOut, nodeColor=<rgba 0.5469 0 1 1>}
  }
  MaterialExpressionNamedRerouteUsage_0 {class=MaterialExpressionNamedRerouteUsage}
  MaterialExpressionNamedRerouteUsage_1 {class=MaterialExpressionNamedRerouteUsage}
  MaterialExpressionNamedRerouteUsage_10 {class=MaterialExpressionNamedRerouteUsage}
  MaterialExpressionNamedRerouteUsage_11 {class=MaterialExpressionNamedRerouteUsage}
  MaterialExpressionNamedRerouteUsage_12 {class=MaterialExpressionNamedRerouteUsage}
  MaterialExpressionNamedRerouteUsage_121 {class=MaterialExpressionNamedRerouteUsage}
  MaterialExpressionNamedRerouteUsage_122 {class=MaterialExpressionNamedRerouteUsage}
  MaterialExpressionNamedRerouteUsage_123 {class=MaterialExpressionNamedRerouteUsage}
  MaterialExpressionNamedRerouteUsage_13 {class=MaterialExpressionNamedRerouteUsage}
  MaterialExpressionNamedRerouteUsage_2 {class=MaterialExpressionNamedRerouteUsage}
  MaterialExpressionNamedRerouteUsage_26 {class=MaterialExpressionNamedRerouteUsage}
  MaterialExpressionNamedRerouteUsage_3 {class=MaterialExpressionNamedRerouteUsage}
  MaterialExpressionNamedRerouteUsage_4 {class=MaterialExpressionNamedRerouteUsage}
  MaterialExpressionNamedRerouteUsage_40 {class=MaterialExpressionNamedRerouteUsage}
  MaterialExpressionNamedRerouteUsage_47 {class=MaterialExpressionNamedRerouteUsage}
  MaterialExpressionNamedRerouteUsage_5 {class=MaterialExpressionNamedRerouteUsage}
  MaterialExpressionNamedRerouteUsage_6 {class=MaterialExpressionNamedRerouteUsage}
  MaterialExpressionNamedRerouteUsage_69 {class=MaterialExpressionNamedRerouteUsage}
  MaterialExpressionNamedRerouteUsage_7 {class=MaterialExpressionNamedRerouteUsage}
  MaterialExpressionNamedRerouteUsage_8 {class=MaterialExpressionNamedRerouteUsage}
  MaterialExpressionNamedRerouteUsage_9 {class=MaterialExpressionNamedRerouteUsage}
  MaterialExpressionNamedRerouteUsage_97 {class=MaterialExpressionNamedRerouteUsage}
  MaterialExpressionNormalize_0 {class=MaterialExpressionNormalize, inputs [VectorInput {from=MaterialExpressionSubtract_8, output=""}]}
  MaterialExpressionOneMinus_1 {class=MaterialExpressionOneMinus, inputs [None {from=MaterialExpressionComponentMask_3}]}
  MaterialExpressionOneMinus_18 {class=MaterialExpressionOneMinus, inputs [None {from=MaterialExpressionReroute_17}]}
  MaterialExpressionOneMinus_3 {class=MaterialExpressionOneMinus, inputs [None {from=MaterialExpressionAbs_7}]}
  MaterialExpressionOneMinus_4 {class=MaterialExpressionOneMinus, inputs [None {from=MaterialExpressionSaturate_6}]}
  MaterialExpressionQualitySwitch_0 {
   class=MaterialExpressionQualitySwitch
   inputs [
    Default {from=MaterialExpressionNamedRerouteUsage_97, output=""}
    Low {from="", output=""}
    High {from="", output=""}
    Medium {from=MaterialExpressionScalarParameter_38, output=""}
    Epic {from="", output=""}
   ]
   properties {}
  }
  MaterialExpressionReroute_12 {class=MaterialExpressionReroute, inputs [None {from=MaterialExpressionReroute_5}]}
  MaterialExpressionReroute_14 {class=MaterialExpressionReroute, inputs [None {from=MaterialExpressionNamedRerouteDeclaration_53}]}
  MaterialExpressionReroute_16 {class=MaterialExpressionReroute, inputs [None {from=MaterialExpressionNamedRerouteDeclaration_46}]}
  MaterialExpressionReroute_17 {class=MaterialExpressionReroute, inputs [None {from=MaterialExpressionNamedRerouteDeclaration_46}]}
  MaterialExpressionReroute_5 {class=MaterialExpressionReroute, inputs [None {from=MaterialExpressionSetMaterialAttributes_0}]}
  MaterialExpressionReroute_6 {class=MaterialExpressionReroute, inputs [None {from=MaterialExpressionSetMaterialAttributes_2}]}
  MaterialExpressionReroute_8 {class=MaterialExpressionReroute, inputs [None {from=MaterialExpressionReroute_5}]}
  MaterialExpressionReroute_9 {class=MaterialExpressionReroute, inputs [None {from=MaterialExpressionNamedRerouteDeclaration_0}]}
  MaterialExpressionRuntimeVirtualTextureOutput_1 {
   class=MaterialExpressionRuntimeVirtualTextureOutput
   inputs [
    BaseColor {from=MaterialExpressionGetMaterialAttributes_8, output=Base Color}
    Specular {from=MaterialExpressionGetMaterialAttributes_8, output=Base Color}
    Roughness {from=MaterialExpressionGetMaterialAttributes_8, output=Base Color}
    Normal {from=MaterialExpressionGetMaterialAttributes_8, output=Base Color}
    WorldHeight {from="", output=""}
    Opacity {from=MaterialExpressionNamedRerouteUsage_47, output=""}
    Mask {from="", output=""}
    Displacement {from="", output=""}
    Mask4 {from="", output=""}
   ]
   properties {}
  }
  MaterialExpressionRuntimeVirtualTextureSample_0 {
   class=MaterialExpressionRuntimeVirtualTextureSample
   inputs [
    UVs {from="", output=""}
    Translated World Position {from="", output=""}
   ]
   properties {
    materialType=BaseColor_Normal_Specular_YCoCg
    virtualTexture=/Game/Athena/Apollo/Maps/Landscape/VT/RVT_Apollo_Landscape
    worldPositionOriginType=CameraRelative
   }
  }
  MaterialExpressionSaturate_13 {class=MaterialExpressionSaturate, inputs [None {from=MaterialExpressionAdd_5}]}
  MaterialExpressionSaturate_16 {class=MaterialExpressionSaturate, inputs [None {from=MaterialExpressionAdd_0}]}
  MaterialExpressionSaturate_17 {class=MaterialExpressionSaturate, inputs [None {from=MaterialExpressionDotProduct_0}]}
  MaterialExpressionSaturate_29 {class=MaterialExpressionSaturate, inputs [None {from=MaterialExpressionMultiply_0}]}
  MaterialExpressionSaturate_4 {class=MaterialExpressionSaturate, inputs [None {from=MaterialExpressionMultiply_22}]}
  MaterialExpressionSaturate_6 {class=MaterialExpressionSaturate, inputs [None {from=MaterialExpressionScalarParameter_14}]}
  MaterialExpressionScalarParameter_0 {class=MaterialExpressionScalarParameter, properties {group=FadeOut, parameterName=InvertFadeDirection_Scalar, sortPriority=-500}}
  MaterialExpressionScalarParameter_1 {class=MaterialExpressionScalarParameter, properties {defaultValue=1, group=Custom01, parameterName=Custom01_EdgeFalloff_Contrast}}
  MaterialExpressionScalarParameter_11 {class=MaterialExpressionScalarParameter, properties {defaultValue=1, group=NarrowFlea, parameterName=NarrowFleaGrassDisplaementScale}}
  MaterialExpressionScalarParameter_12 {
   class=MaterialExpressionScalarParameter
   properties {defaultValue=2, group=NarrowFlea, parameterName=NarrowFleaStripeScale, primitiveDataIndex=6}
  }
  MaterialExpressionScalarParameter_14 {class=MaterialExpressionScalarParameter, properties {defaultValue=0.75, group=NarrowFlea, parameterName=NarrowFleaStripeSharpness}}
  MaterialExpressionScalarParameter_2 {class=MaterialExpressionScalarParameter, properties {group=FadeOut, parameterName=UseFadeOut, sortPriority=0}}
  MaterialExpressionScalarParameter_3 {class=MaterialExpressionScalarParameter, properties {parameterName=IsDirtPath}}
  MaterialExpressionScalarParameter_38 {
   class=MaterialExpressionScalarParameter
   properties {
    defaultValue=1
    desc="A temp fix for Mobile rendering bug where static 1.0 in opacity + WPO is causing it to render black.  Serge Bernier is looking into a permanent fix and then we can swap this back to a static 1.0 if desired."
    group=ZDoNotSet
    parameterName=SwitchMaskOpacity
    sliderMax=1
    sliderMin=1
   }
  }
  MaterialExpressionSetMaterialAttributes_0 {
   class=MaterialExpressionSetMaterialAttributes
   inputs [
    MaterialAttributes {from=MaterialExpressionMaterialFunctionCall_0, output=Result}
    Subsurface Color {from=MaterialExpressionMultiply_4, output=""}
   ]
   properties {attributeSetTypes [#6]}
  }
  MaterialExpressionSetMaterialAttributes_15 {
   class=MaterialExpressionSetMaterialAttributes
   inputs [
    MaterialAttributes {from="", output=""}
    Base Color {from=MaterialExpressionRuntimeVirtualTextureSample_0, output=BaseColor}
    Specular {from=MaterialExpressionRuntimeVirtualTextureSample_0, output=BaseColor}
    Roughness {from=MaterialExpressionRuntimeVirtualTextureSample_0, output=BaseColor}
    Normal {from=MaterialExpressionRuntimeVirtualTextureSample_0, output=BaseColor}
   ]
   properties {attributeSetTypes [
     #1
     #3
     #4
     #5
    ]}
  }
  MaterialExpressionSetMaterialAttributes_2 {
   class=MaterialExpressionSetMaterialAttributes
   inputs [
    MaterialAttributes {from=MaterialExpressionReroute_8, output=""}
    Opacity Mask {from=MaterialExpressionQualitySwitch_0, output=""}
    World Position Offset {from=MaterialExpressionShadowReplace_1, output=""}
   ]
   properties {
    desc="The opacity mask is used on mobile - this material is using Masked blend mode!"
    attributeSetTypes [#2, #7]
   }
  }
  MaterialExpressionSetMaterialAttributes_25 {
   class=MaterialExpressionSetMaterialAttributes
   inputs [
    MaterialAttributes {from=MaterialExpressionReroute_12, output=""}
    Opacity Mask {from=MaterialExpressionConstant_24, output=""}
   ]
   properties {desc="Masked blend mode is enabled, but only mobile utilises it!", attributeSetTypes [#2]}
  }
  MaterialExpressionSetMaterialAttributes_4 {
   class=MaterialExpressionSetMaterialAttributes
   inputs [
    MaterialAttributes {from=MaterialExpressionGetMaterialAttributes_3, output=MaterialAttributes}
    Base Color {from=MaterialExpressionSaturate_4, output=""}
    Displacement {from="", output=""}
   ]
   properties {attributeSetTypes [#1, #8]}
  }
  MaterialExpressionShadingPathSwitch_0 {
   class=MaterialExpressionShadingPathSwitch
   inputs [
    Default {from=MaterialExpressionMaterialFunctionCall_22, output=Alpha}
    Deferred {from="", output=""}
    Forward {from="", output=""}
    Mobile {from=MaterialExpressionMaterialFunctionCall_22, output=Alpha}
   ]
   properties {}
  }
  MaterialExpressionShadingPathSwitch_3 {
   class=MaterialExpressionShadingPathSwitch
   inputs [
    Default {from=MaterialExpressionSetMaterialAttributes_25, output=""}
    Deferred {from="", output=""}
    Forward {from="", output=""}
    Mobile {from=MaterialExpressionSetMaterialAttributes_2, output=""}
   ]
   properties {}
  }
  MaterialExpressionShadowReplace_1 {
   class=MaterialExpressionShadowReplace
   inputs [
    Default {from=MaterialExpressionMultiply_23, output=""}
    Shadow {from=MaterialExpressionConstant3Vector_3, output=""}
   ]
   properties {}
  }
  MaterialExpressionSmoothStep_0 {
   class=MaterialExpressionSmoothStep
   inputs [
    Min {from="", output=""}
    Max {from="", output=""}
    Value {from=MaterialExpressionNamedRerouteUsage_4, output=""}
   ]
   properties {constMax=0.05}
  }
  MaterialExpressionSmoothStep_1 {
   class=MaterialExpressionSmoothStep
   inputs [
    Min {from=MaterialExpressionAdd_6, output=""}
    Max {from=MaterialExpressionSubtract_0, output=""}
    Value {from=MaterialExpressionAbs_8, output=""}
   ]
   properties {constMax=0.45, constMin=0.55}
  }
  MaterialExpressionSmoothStep_2 {
   class=MaterialExpressionSmoothStep
   inputs [
    Min {from=MaterialExpressionVectorParameter_0, output=FalloffMin}
    Max {from=MaterialExpressionVectorParameter_0, output=FalloffMin}
    Value {from=MaterialExpressionReroute_14, output=""}
   ]
   properties {constMax=0.5}
  }
  MaterialExpressionSmoothStep_3 {
   class=MaterialExpressionSmoothStep
   inputs [
    Min {from=MaterialExpressionVectorParameter_4, output=FalloffMin}
    Max {from=MaterialExpressionVectorParameter_4, output=FalloffMin}
    Value {from=MaterialExpressionReroute_14, output=""}
   ]
   properties {constMax=0.5}
  }
  MaterialExpressionSmoothStep_4 {
   class=MaterialExpressionSmoothStep
   inputs [
    Min {from="", output=""}
    Max {from="", output=""}
    Value {from=MaterialExpressionSquareRoot_0, output=""}
   ]
   properties {constMax=0.5, constMin=0.1}
  }
  MaterialExpressionSquareRoot_0 {class=MaterialExpressionSquareRoot, inputs [None {from=MaterialExpressionNamedRerouteUsage_4, output=""}], properties {}}
  MaterialExpressionStaticBoolParameter_3 {
   class=MaterialExpressionStaticBoolParameter
   inputs []
   properties {
    desc="Default=False. True will use the heightmap of a defined landscape layer for the falloff blend. Not used on mobile."
    group=Custom01
    parameterName=Use_FalloffToLandscapeLayerMatch
    sortPriority=-5000
   }
  }
  MaterialExpressionStaticBool_0 {class=MaterialExpressionStaticBool}
  MaterialExpressionStaticSwitchParameter_1 {
   class=MaterialExpressionStaticSwitchParameter
   inputs [
    True {from=MaterialExpressionNamedRerouteUsage_69, output=""}
    False {from=MaterialExpressionNamedRerouteDeclaration_0, output=""}
   ]
   properties {
    defaultValue=true
    desc="This should be set to false for roads on things like bridges, where RVT blending is not desired"
    group=RVT
    parameterName=ReadRVT
   }
  }
  MaterialExpressionStaticSwitch_6 {
   class=MaterialExpressionStaticSwitch
   inputs [
    True {from=MaterialExpressionMaterialFunctionCall_20, output=Alpha}
    False {from=MaterialExpressionMaterialFunctionCall_22, output=Alpha}
    Value {from=MaterialExpressionNamedRerouteUsage_1, output=""}
   ]
   properties {}
  }
  MaterialExpressionSubtract_0 {
   class=MaterialExpressionSubtract
   inputs [
    A {from=""}
    B {from=MaterialExpressionMax_1}
   ]
   properties {constA=0.5}
  }
  MaterialExpressionSubtract_8 {
   class=MaterialExpressionSubtract
   inputs [
    A {from=MaterialExpressionCameraPositionWS_0}
    B {from=MaterialExpressionWorldPosition_1, output=XYZ}
   ]
  }
  MaterialExpressionTextureCoordinate_0 {class=MaterialExpressionTextureCoordinate, properties {coordinateIndex=1}}
  MaterialExpressionTextureCoordinate_1 {class=MaterialExpressionTextureCoordinate}
  MaterialExpressionTextureCoordinate_4 {class=MaterialExpressionTextureCoordinate}
  MaterialExpressionTextureCoordinate_5 {class=MaterialExpressionTextureCoordinate}
  MaterialExpressionTextureSampleParameter2D_7 {
   class=MaterialExpressionTextureSampleParameter2D
   inputs [
    UVs {from=MaterialExpressionTextureCoordinate_5, output=""}
    Apply View MipBias {from="", output=""}
   ]
   properties {
    desc=CLAMPED
    group=RoadTypeEdgeFalloff
    parameterName=RoadTypeEdgeFalloffMask
    sortPriority=-500
    texture=/Game/Athena/Apollo/Environments/Landscape/Textures/Roads/Helios/T_RoadFalloffMasks
   }
  }
  MaterialExpressionTextureSample_0 {
   class=MaterialExpressionTextureSample
   inputs [
    UVs {from=""}
    Tex {from=""}
    Apply View MipBias {from=""}
   ]
   properties {texture=/Hera/Terrain/Cliffs/Textures/T_CosmicThunder_RoadStripes}
  }
  MaterialExpressionTextureSample_1 {
   class=MaterialExpressionTextureSample
   inputs [
    UVs {from=""}
    Tex {from=""}
    Apply View MipBias {from=""}
   ]
   properties {
    samplerSource=SSM_Clamp_WorldGroupSettings
    samplerType=SAMPLERTYPE_LinearColor
    texture=&T_CosmicThunder_RoadStripes
   }
  }
  MaterialExpressionVectorParameter_0 {
   class=MaterialExpressionVectorParameter
   properties {
    group=RoadTypeEdgeFalloff
    parameterName=RoadTypeEdgeFalloff_Large
    sortPriority=-300
    channelNames {a=Add, b=Multiply, g=FalloffMax, r=FalloffMin}
    defaultValue=<rgba -0.75 0.25 1 0>
   }
  }
  MaterialExpressionVectorParameter_4 {
   class=MaterialExpressionVectorParameter
   properties {
    group=RoadTypeEdgeFalloff
    parameterName=RoadType_EdgeFalloff_Small
    sortPriority=-200
    channelNames {a=Add, b=Multiply, g=FalloffMax, r=FalloffMin}
    defaultValue=<rgba -0.01 0.01 1 0>
   }
  }
  MaterialExpressionVectorParameter_8 {
   class=MaterialExpressionVectorParameter
   properties {
    desc="Default is setup as R=Straight, G=JunctionT, B=JunctionY, A=JunctionSplit"
    group=RoadTypeEdgeFalloff
    parameterName=RoadTypeEdgeFalloff_Channel
    sortPriority=-400
    defaultValue=<rgba 1 0 0 0>
   }
  }
  MaterialExpressionWorldPosition_1 {class=MaterialExpressionWorldPosition, inputs [], properties {}}
 }
 properties {
  bAutomaticallySetUsageInEditor=false
  bEnableDisplacementFade=true
  bEnableTessellation=true
  bTangentSpaceNormal=false
  bUseMaterialAttributes=true
  bUsedWithNanite=true
  bUsedWithSplineMeshes=true
  shadingModel=MSM_Subsurface
  usedShadingModels=Subsurface
  displacementFadeRange {endSizePixels=6, startSizePixels=8}
  displacementScaling {center=0, magnitude=10}
 }
}
